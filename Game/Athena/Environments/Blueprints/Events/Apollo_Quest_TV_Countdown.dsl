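; Event Blueprint that, judging by its name, drives a quest countdown on in-game TVs.
; On BeginPlay it watches a calendar event, whose name is blank here, and when active derives normalized progress from time remaining and enables the countdown.
; OnRep_ShowCountdown pushes visibility and progress to the shared MPC_Countdown material parameter collection, or clears visibility when disabled.
; OnRep_NormalizedProgess is empty, so progress reaches the material only through the visibility callback.
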(event EventBeginPlay
  (bind (_eventname _timeuntilend) (MCPContext|WatchCalendarEvent "")
    (:OnEventActive
      (Variables|Default|SetNormalizedProgess (/ (Math|Conversions|ToFloat(Integer) (Math|Timespan|BreakTimespan _timeuntilend)) 4.0))
      (Variables|Default|SetShowCountdown true))))

(fn OnRep_NormalizedProgess)

(fn OnRep_ShowCountdown
  (if (Variables|Default|GetShowCountdown)
    (Rendering|Material|SetScalarParameterValue "/Game/Athena/Events/Generic/CountdownTimer/Materials/MPC_Countdown.MPC_Countdown" "ShowCountdown" 1.0 :!class /Script/Engine.KismetMaterialLibrary)
    (Rendering|Material|SetScalarParameterValue "/Game/Athena/Events/Generic/CountdownTimer/Materials/MPC_Countdown.MPC_Countdown" "ProgressNormalized" (Variables|Default|GetNormalizedProgess) :!class /Script/Engine.KismetMaterialLibrary)
    (else
      (Rendering|Material|SetScalarParameterValue "/Game/Athena/Events/Generic/CountdownTimer/Materials/MPC_Countdown.MPC_Countdown" "ShowCountdown" :!class /Script/Engine.KismetMaterialLibrary))))

(fn ConstructionScript)
